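# Material function for the BisonDrain TV gadget prop that computes one bloom tap of an animated flipbook texture, apparently built for chaining.
# It samples FlipBook three times at horizontally staggered UV offsets, keeping one of R, G and B from each for a color-split smear.
# The dimmed result is added to a color, with outputs for an incremented offset, UVs, grid sizes and time multiplier mirroring the inputs.
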MaterialFunction /BRCosmetics/Animation/Props/Gadgets_BisonDrain_TV/Materials/MF_BloomFlipbookBisonDrain
graph {nodes [
  n1 = Multiply(A=NamedRerouteUsage()) {constB=2}
  n2 = NamedRerouteUsage()
  n3 = NamedRerouteUsage()
  n4 = NamedRerouteUsage()
  n5 = NamedRerouteUsage()
  n6 = Constant()
  n7 = Add(A=NamedRerouteUsage(), B=Multiply(A=AppendVector(A=@n5, B=@n6)) {constB=0.007})
  n8 = Multiply(A=NamedRerouteUsage()) {constB=2}
  n9 = NamedRerouteUsage()
  n10 = NamedRerouteUsage()
  n11 = NamedRerouteUsage()
  n12 = Subtract(A=NamedRerouteUsage()) {constB=2}
  n13 = Constant()
  n14 = Add(A=NamedRerouteUsage(), B=Multiply(A=AppendVector(A=@n12, B=@n13)) {constB=0.007})
  n15 = Subtract(A=Multiply(A=NamedRerouteUsage()) {constB=2})
  n16 = NamedRerouteUsage()
  n17 = NamedRerouteUsage()
  n18 = NamedRerouteUsage()
  n19 = Subtract(A=NamedRerouteUsage()) {constB=4}
  n20 = Constant()
  n21 = Add(A=NamedRerouteUsage(), B=Multiply(A=AppendVector(A=@n19, B=@n20)) {constB=0.007})
  n22 = MaterialFunctionCall(
   R=ComponentMask(None=MaterialFunctionCall("Animation  Phase (0-1)"=@n1, Number of Rows=@n2, Number of Columns=@n3, Texture=@n4, UVs=@n7) {materialFunction=&FlipBook}.Result) {r=true}
   G=ComponentMask(None=MaterialFunctionCall("Animation  Phase (0-1)"=@n8, Number of Rows=@n9, Number of Columns=@n10, Texture=@n11, UVs=@n14) {materialFunction=&FlipBook}.Result) {g=true}
   B=ComponentMask(None=MaterialFunctionCall("Animation  Phase (0-1)"=@n15, Number of Rows=@n16, Number of Columns=@n17, Texture=@n18, UVs=@n21) {materialFunction=&FlipBook}.Result) {b=true}
  ) {materialFunction=&AppendMany}
  FunctionOutput(None=Add(A=Multiply(A=@n22.RGB) {constB=0.1}, B=NamedRerouteUsage())) {outputName=OUT_ColorBloom}
  FunctionOutput(None=NamedRerouteUsage()) {outputName=OUT_RowSize, sortPriority=3}
  FunctionOutput(None=NamedRerouteUsage()) {outputName=OUT_UVn, sortPriority=1}
  FunctionOutput(None=Add(A=NamedRerouteUsage())) {outputName=OUT_x, sortPriority=2}
  FunctionOutput(None=NamedRerouteUsage()) {outputName=OUT_ColumnSize, sortPriority=4}
  FunctionOutput(None=NamedRerouteUsage()) {outputName=OUT_FlipbookTimeMultiplier, sortPriority=5}
  NamedRerouteDeclaration(None=FunctionInput(Preview=TextureObject()) {bUsePreviewValueAsDefault=true, inputName=IN_FlipbookToBloom, inputType=FunctionInput_Texture2D, sortPriority=3}) {name=Flipbook, nodeColor=<rgba 0.9219 0 1 1>}
  NamedRerouteDeclaration(None=FunctionInput(Preview=TextureCoordinate()) {bUsePreviewValueAsDefault=true, inputName=IN_UVn, inputType=FunctionInput_Vector2, sortPriority=1}) {name=UVn, nodeColor=<rgba 0 1 0.6016 1>}
  NamedRerouteDeclaration(None=FunctionInput(Preview=Constant() {r=-4}) {bUsePreviewValueAsDefault=true, inputName=IN_x, inputType=FunctionInput_Scalar, sortPriority=2}) {name=IN_x, nodeColor=<rgba 1 0.2812 0 1>}
  NamedRerouteDeclaration(None=FunctionInput(Preview=Constant3Vector() {constant=<rgba 0.9 0.1 0.14 0>}) {bUsePreviewValueAsDefault=true, inputName=IN_Color}) {name=IN_Color, nodeColor=<rgba 1 0 0.4922 1>}
  n23 = Time()
  n24 = NamedRerouteUsage()
  NamedRerouteDeclaration(None=Frac(None=Multiply(A=@n23, B=@n24) {constB=0.1111})) {name=AnimationPhase, nodeColor=<rgba 0.5937 1 0 1>}
  NamedRerouteDeclaration(None=FunctionInput(Preview=Constant() {r=8}) {bUsePreviewValueAsDefault=true, inputName=IN_RowSize, inputType=FunctionInput_Scalar, sortPriority=4}) {name=RowSize, nodeColor=<rgba 1 0 0.9141 1>}
  NamedRerouteDeclaration(None=FunctionInput(Preview=Constant() {r=16}) {bUsePreviewValueAsDefault=true, inputName=IN_ColumnSize, inputType=FunctionInput_Scalar, sortPriority=5}) {name=ColumnSize, nodeColor=<rgba 0 1 0.7656 1>}
  n25 = AppendVector(A=NamedRerouteUsage(), B=Constant())
  NamedRerouteDeclaration(None=Add(A=NamedRerouteUsage(), B=Multiply(A=@n25) {constB=0.007})) {name=OffsetUV, nodeColor=<rgba 0 1 0.3438 1>}
  NamedRerouteDeclaration(None=FunctionInput(Preview=Constant() {r=0.1111}) {bUsePreviewValueAsDefault=true, inputName=IN_FlipbookTimeMultiplier, inputType=FunctionInput_Scalar, sortPriority=6}) {name=FlipbookTimeMultiplier, nodeColor=<rgba 0.8047 1 0 1>}
 ]}
