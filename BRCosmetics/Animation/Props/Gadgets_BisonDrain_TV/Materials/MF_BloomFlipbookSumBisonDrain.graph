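# Material function for the BisonDrain TV gadget screen that builds a bloomed color from a flipbook texture.
# It chains calls to MF_BloomFlipbookBisonDrain, seeding the first with color, UV, an x value, grid size and time multiplier.
# Scalar parameters scale the color before the chain and after the fifth call, which feeds BloomedColor.
# Two further chained calls exist but do not reach the output.
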MaterialFunction /BRCosmetics/Animation/Props/Gadgets_BisonDrain_TV/Materials/MF_BloomFlipbookSumBisonDrain
graph {nodes [
  n1 = TextureSample(Tex=TextureObject())
  n2 = TextureObject()
  n3 = MaterialFunctionCall!(
   IN_Color=Multiply(
    A=FunctionInput(Preview=@n1.RGB) {inputName=IN_Color}
    B=ScalarParameter() {defaultValue=1, group=BisonDrainScreen, parameterName=MultiplyInputColorBeforeBloom}
   )
   IN_UVn=FunctionInput(Preview=TextureCoordinate()) {inputName=IN_UV, inputType=FunctionInput_Vector2, sortPriority=1}
   IN_x=FunctionInput(Preview=Constant() {r=-2}) {inputName=IN_x, inputType=FunctionInput_Scalar, sortPriority=2}
   IN_FlipbookToBloom=NamedRerouteDeclaration(None=FunctionInput(Preview=@n2) {inputName=IN_Flipbook, inputType=FunctionInput_Texture2D, sortPriority=3}) {name=IN_Flipbook, nodeColor=<rgba 1 0.2109 0 1>}
   IN_RowSize=NamedRerouteUsage()
   IN_ColumnSize=NamedRerouteUsage()
   IN_FlipbookTimeMultiplier=NamedRerouteUsage()
  ) {materialFunction=&/MF_BloomFlipbookBisonDrain}
  n4 = MaterialFunctionCall(
   IN_Color=@n3.OUT_ColorBloom
   IN_UVn=@n3.OUT_ColorBloom
   IN_x=@n3.OUT_ColorBloom
   IN_FlipbookToBloom=NamedRerouteUsage()
   IN_RowSize=@n3.OUT_ColorBloom
   IN_ColumnSize=@n3.OUT_ColorBloom
   IN_FlipbookTimeMultiplier=@n3.OUT_ColorBloom
  ) {materialFunction=&MF_BloomFlipbookBisonDrain}
  n5 = MaterialFunctionCall(
   IN_Color=@n4.OUT_ColorBloom
   IN_UVn=@n4.OUT_ColorBloom
   IN_x=@n4.OUT_ColorBloom
   IN_FlipbookToBloom=NamedRerouteUsage()
   IN_RowSize=@n4.OUT_ColorBloom
   IN_ColumnSize=@n4.OUT_ColorBloom
   IN_FlipbookTimeMultiplier=@n4.OUT_ColorBloom
  ) {materialFunction=&MF_BloomFlipbookBisonDrain}
  n6 = MaterialFunctionCall(
   IN_Color=@n5.OUT_ColorBloom
   IN_UVn=@n5.OUT_ColorBloom
   IN_x=@n5.OUT_ColorBloom
   IN_FlipbookToBloom=NamedRerouteUsage()
   IN_RowSize=@n5.OUT_ColorBloom
   IN_ColumnSize=@n5.OUT_ColorBloom
   IN_FlipbookTimeMultiplier=@n5.OUT_ColorBloom
  ) {materialFunction=&MF_BloomFlipbookBisonDrain}
  n7 = MaterialFunctionCall(
   IN_Color=@n6.OUT_ColorBloom
   IN_UVn=@n6.OUT_ColorBloom
   IN_x=@n6.OUT_ColorBloom
   IN_FlipbookToBloom=NamedRerouteUsage()
   IN_RowSize=@n6.OUT_ColorBloom
   IN_ColumnSize=@n6.OUT_ColorBloom
   IN_FlipbookTimeMultiplier=@n6.OUT_ColorBloom
  ) {materialFunction=&MF_BloomFlipbookBisonDrain}
  FunctionOutput(None=Multiply(A=@n7.OUT_ColorBloom, B=ScalarParameter() {defaultValue=0.6, group=BisonDrainScreen, parameterName=MultiplyInputColorAfterBloom}) {constB=0.6}) {outputName=BloomedColor}
  n8 = MaterialFunctionCall(
   IN_Color=@n7.OUT_ColorBloom
   IN_UVn=@n7.OUT_ColorBloom
   IN_x=@n7.OUT_ColorBloom
   IN_FlipbookToBloom=NamedRerouteUsage()
   IN_RowSize=@n7.OUT_ColorBloom
   IN_ColumnSize=@n7.OUT_ColorBloom
   IN_FlipbookTimeMultiplier=@n7.OUT_ColorBloom
  ) {materialFunction=&MF_BloomFlipbookBisonDrain}
  MaterialFunctionCall(
   IN_Color=@n8.OUT_ColorBloom
   IN_UVn=@n8.OUT_ColorBloom
   IN_x=@n8.OUT_ColorBloom
   IN_FlipbookToBloom=NamedRerouteUsage()
   IN_RowSize=@n8.OUT_ColorBloom
   IN_ColumnSize=@n8.OUT_ColorBloom
   IN_FlipbookTimeMultiplier=@n8.OUT_ColorBloom
  ) {materialFunction=&MF_BloomFlipbookBisonDrain}
  NamedRerouteDeclaration(None=FunctionInput(Preview=Constant() {r=8}) {inputName=IN_RowSize, inputType=FunctionInput_Scalar, sortPriority=4, previewValue=<8 0 0 1>}) {name=RowSize, nodeColor=<rgba 0.1016 1 0 1>}
  n9 = FunctionInput(Preview=Constant() {r=16}) {inputName=IN_ColumnSize, inputType=FunctionInput_Scalar, sortPriority=5, previewValue=<16 0 0 1>}
  NamedRerouteDeclaration(None=@n9) {name=ColumnSize, nodeColor=<rgba 0.6406 1 0 1>}
  n10 = FunctionInput(Preview=Constant() {r=0.1111}) {inputName=IN_FlipbookTimeMultiplier, inputType=FunctionInput_Scalar, sortPriority=6, previewValue=<16 0 0 1>}
  NamedRerouteDeclaration(None=@n10) {name=FlipbookTimeMultiplier, nodeColor=<rgba 0.6406 1 0 1>}
 ]}
