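# Material function for the BisonDrain TV gadget prop that accumulates one tap of a horizontal, color-fringed bloom from a flipbook texture.
# It samples the flipbook three times at staggered U offsets derived from a step index, taking red, green and blue from separate taps.
# The recombined RGB is scaled down and added onto a color.
# It also outputs UVs and an incremented index, which suggests instances are chained to build up the blur.
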MaterialFunction /BRCosmetics/Animation/Props/Gadgets_BisonDrain_TV/Materials/MF_BloomBisonDrain
graph {nodes [
  n1 = NamedRerouteUsage()
  n2 = Multiply(A=AppendVector(A=NamedRerouteUsage(), B=Constant())) {constB=0.007}
  n3 = NamedRerouteUsage()
  n4 = NamedRerouteUsage()
  n5 = Multiply(A=AppendVector(A=Subtract(A=@n4) {constB=2}, B=Constant())) {constB=0.007}
  n6 = NamedRerouteUsage()
  n7 = NamedRerouteUsage()
  n8 = Multiply(A=AppendVector(A=Subtract(A=@n7) {constB=4}, B=Constant())) {constB=0.007}
  n9 = MaterialFunctionCall(
   R=TextureSample(UVs=Add(A=@n1, B=@n2), Tex=NamedRerouteUsage()) {samplerSource=SSM_Wrap_WorldGroupSettings}.R
   G=TextureSample(UVs=Add(A=@n3, B=@n5), Tex=NamedRerouteUsage()) {samplerSource=SSM_Wrap_WorldGroupSettings}.G
   B=TextureSample(UVs=Add(A=@n6, B=@n8), Tex=NamedRerouteUsage()) {samplerSource=SSM_Wrap_WorldGroupSettings}.B
  ) {materialFunction=&AppendMany}
  FunctionOutput(None=Add(A=Multiply(A=@n9.RGB) {constB=0.1}, B=NamedRerouteUsage())) {outputName=OUT_ColorBloom}
  FunctionOutput(None=NamedRerouteUsage()) {outputName=OUT_UVn, sortPriority=1}
  FunctionOutput(None=Add(A=NamedRerouteUsage())) {outputName=OUT_x, sortPriority=2}
  NamedRerouteDeclaration(None=FunctionInput(Preview=TextureObject()) {bUsePreviewValueAsDefault=true, inputName=IN_FlipbookToBloom, inputType=FunctionInput_Texture2D, sortPriority=3}) {name=Flipbook, nodeColor=<rgba 0.9219 0 1 1>}
  NamedRerouteDeclaration(None=FunctionInput(Preview=TextureCoordinate()) {bUsePreviewValueAsDefault=true, inputName=IN_UVn, inputType=FunctionInput_Vector2, sortPriority=1}) {name=UVn, nodeColor=<rgba 0 1 0.6016 1>}
  NamedRerouteDeclaration(None=FunctionInput(Preview=Constant() {r=-4}) {inputName=IN_x, inputType=FunctionInput_Scalar, sortPriority=2}) {name=IN_x, nodeColor=<rgba 1 0.2812 0 1>}
  NamedRerouteDeclaration(None=FunctionInput(Preview=Constant3Vector() {constant=<rgba 0.9 0.1 0.14 0>}) {inputName=IN_Color}) {name=IN_Color, nodeColor=<rgba 1 0 0.4922 1>}
 ]}
