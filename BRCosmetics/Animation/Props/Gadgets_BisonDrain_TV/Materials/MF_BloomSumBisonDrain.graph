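# Material function that builds a bloomed color for the BisonDrain TV gadget screen in a cosmetic animation prop.
# It chains seven MF_BloomBisonDrain calls, each feeding the next, starting from input color, UV and x value.
# A named reroute shares the input flipbook texture with every pass.
# Scalar parameters in the BisonDrainScreen group scale the color before and after the chain, producing BloomedColor.
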MaterialFunction /BRCosmetics/Animation/Props/Gadgets_BisonDrain_TV/Materials/MF_BloomSumBisonDrain
graph {nodes [
  n1 = TextureSample(Tex=TextureObject())
  n2 = TextureObject()
  n3 = MaterialFunctionCall!(
   IN_Color=Multiply(
    A=FunctionInput(Preview=@n1.RGB) {inputName=IN_Color}
    B=ScalarParameter() {defaultValue=1, group=BisonDrainScreen, parameterName=MultiplyInputColorBeforeBloom}
   )
   IN_UVn=FunctionInput(Preview=TextureCoordinate()) {inputName=IN_UV, inputType=FunctionInput_Vector2, sortPriority=1}
   IN_x=FunctionInput(Preview=Constant() {r=-4}) {inputName=IN_x, inputType=FunctionInput_Scalar, sortPriority=2}
   IN_FlipbookToBloom=NamedRerouteDeclaration(None=FunctionInput(Preview=@n2) {inputName=IN_Flipbook, inputType=FunctionInput_Texture2D, sortPriority=3}) {name=IN_Flipbook, nodeColor=<rgba 1 0.2109 0 1>}
  ) {materialFunction=&/MF_BloomBisonDrain}
  n4 = MaterialFunctionCall(IN_Color=@n3.OUT_ColorBloom, IN_UVn=@n3.OUT_ColorBloom, IN_x=@n3.OUT_ColorBloom, IN_FlipbookToBloom=NamedRerouteUsage()) {materialFunction=&MF_BloomBisonDrain}
  n5 = MaterialFunctionCall(IN_Color=@n4.OUT_ColorBloom, IN_UVn=@n4.OUT_ColorBloom, IN_x=@n4.OUT_ColorBloom, IN_FlipbookToBloom=NamedRerouteUsage()) {materialFunction=&MF_BloomBisonDrain}
  n6 = MaterialFunctionCall(IN_Color=@n5.OUT_ColorBloom, IN_UVn=@n5.OUT_ColorBloom, IN_x=@n5.OUT_ColorBloom, IN_FlipbookToBloom=NamedRerouteUsage()) {materialFunction=&MF_BloomBisonDrain}
  n7 = MaterialFunctionCall(IN_Color=@n6.OUT_ColorBloom, IN_UVn=@n6.OUT_ColorBloom, IN_x=@n6.OUT_ColorBloom, IN_FlipbookToBloom=NamedRerouteUsage()) {materialFunction=&MF_BloomBisonDrain}
  n8 = MaterialFunctionCall(IN_Color=@n7.OUT_ColorBloom, IN_UVn=@n7.OUT_ColorBloom, IN_x=@n7.OUT_ColorBloom, IN_FlipbookToBloom=NamedRerouteUsage()) {materialFunction=&MF_BloomBisonDrain}
  n9 = NamedRerouteUsage()
  FunctionOutput(None=Multiply(
    A=MaterialFunctionCall(IN_Color=@n8.OUT_ColorBloom, IN_UVn=@n8.OUT_ColorBloom, IN_x=@n8.OUT_ColorBloom, IN_FlipbookToBloom=@n9) {materialFunction=&MF_BloomBisonDrain}.OUT_ColorBloom
    B=ScalarParameter() {defaultValue=0.6, group=BisonDrainScreen, parameterName=MultiplyInputColorAfterBloom}
   ) {constB=0.6}) {outputName=BloomedColor}
 ]}
