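# Library material function for the BisonDrain TV gadget prop that samples a texture and sets a minimum color based on horizontal UV.
# It takes texture and UV inputs, samples once, and measures the U coordinate's distance from the 0.5 center.
# An If node substitutes constant dark grey where U falls outside the 0–1 range, apparently masking warped UV overflow at the left and right edges.
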MaterialFunction /BRCosmetics/Animation/Props/Gadgets_BisonDrain_TV/Materials/MF_WarpedTextureSample
graph {
 properties {bExposeToLibrary=true, description=Sets a min value depending on X UV value, userExposedCaption=Sets a min value depending on X UV value}
 nodes [
  FunctionOutput(None=NamedRerouteUsage())
  n1 = FunctionInput(Preview=TextureObject()) {bUsePreviewValueAsDefault=true, inputName=IN_Texture, inputType=FunctionInput_Texture2D, previewValue=<1 0 0 1>}
  NamedRerouteDeclaration(None=@n1) {name=IN_Texture, nodeColor=<rgba 0 0.2266 1 1>}
  n2 = FunctionInput(Preview=TextureCoordinate()) {bUsePreviewValueAsDefault=true, inputName=IN_UVs, inputType=FunctionInput_Vector2, previewValue=<0.5 0.5 0 1>}
  NamedRerouteDeclaration(None=@n2) {name=IN_UVs, nodeColor=<rgba 0.5469 1 0 1>}
  NamedRerouteDeclaration(None=TextureSample(UVs=NamedRerouteUsage(), Tex=NamedRerouteUsage()).RGB) {name=SampledTexture, nodeColor=<rgba 1 0 0.8672 1>}
  n3 = Subtract(A=ComponentMask(None=NamedRerouteUsage()) {r=true}) {constB=0.5}
  n4 = NamedRerouteUsage()
  NamedRerouteDeclaration(None=If(A=Constant() {r=0.5}, B=Abs(None=@n3), "A > B"=@n4, "A == B"=@n4, "A < B"=Constant3Vector() {constant=<rgba 0.1 0.1 0.1 0>})) {name=MinColor, nodeColor=<rgba 0 1 0.8125 1>}
 ]
}
