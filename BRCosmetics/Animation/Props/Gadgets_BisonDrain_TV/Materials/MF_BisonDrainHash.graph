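# Material function that hashes a 2D vector into a pseudo-random float, stored with the BisonDrain TV gadget prop materials and exposed to the material library.
# It uses the classic frac-sine hash: dot the input with a constant vector, take the sine, scale by a large constant and keep the fractional part.
# The input previews with texture coordinates and is passed through a named reroute.
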MaterialFunction /BRCosmetics/Animation/Props/Gadgets_BisonDrain_TV/Materials/MF_BisonDrainHash
graph {
 properties {bExposeToLibrary=true, description=Hash from 2d to float, userExposedCaption=Hash from 2d to float}
 nodes [
  n1 = Sine(None=DotProduct(A=NamedRerouteUsage(), B=Constant2Vector() {g=19.36, r=89.44}))
  FunctionOutput(None=Frac(None=Multiply(A=@n1) {constB=22189.2207})) {outputName=OUT_HashFloat}
  NamedRerouteDeclaration(None=FunctionInput(Preview=TextureCoordinate()) {inputName=IN_2dVector, inputType=FunctionInput_Vector2}) {name=IN_V, nodeColor=<rgba 0 0.7891 1 1>}
 ]
}
