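# Material function in the Bison Drain TV gadget prop's materials that outputs a ratio of two powers of two.
# Scalar input `i`, previewed with the UV red channel, feeds the only named reroute declaration.
# Two reroute usages, with no visible link, supply the exponents for identical base-2 Power nodes, which are then divided.
# If both usages resolve to `i`, the result is a constant one.
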MaterialFunction /BRCosmetics/Animation/Props/Gadgets_BisonDrain_TV/Materials/MF_PowerMultBisonDrain
graph {nodes [
  n1 = Constant() {r=2}
  n2 = NamedRerouteUsage()
  n3 = Constant() {r=2}
  n4 = NamedRerouteUsage()
  FunctionOutput(None=Divide(A=Power(Base=@n1, Exp=@n2), B=Power(Base=@n3, Exp=@n4)))
  n5 = TextureCoordinate()
  NamedRerouteDeclaration(None=FunctionInput(Preview=ComponentMask(None=@n5) {r=true}) {inputName=i, inputType=FunctionInput_Scalar}) {name=i, nodeColor=<rgba 0 0.4141 1 1>}
 ]}
