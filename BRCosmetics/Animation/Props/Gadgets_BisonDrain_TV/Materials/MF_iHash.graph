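# Material function for the BisonDrain TV gadget prop that computes iHash, an interpolated 2D hash in the style of value noise.
# Vector2 inputs IN_V and IN_R, defaulting to texture coordinates, support a grid where four floored, offset corners are divided and hashed by MF_BisonDrainHash.
# Named reroutes carry corner hashes and fractional position into a bilinear-style weighted sum.
# A SmoothStep on the fraction is left unconnected, so interpolation is unsmoothed.
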MaterialFunction /BRCosmetics/Animation/Props/Gadgets_BisonDrain_TV/Materials/MF_iHash
graph {nodes [
  n1 = NamedRerouteUsage()
  n2 = NamedRerouteUsage()
  n3 = NamedRerouteUsage()
  n4 = NamedRerouteUsage()
  n5 = Multiply(A=Add(A=Multiply(A=@n1, B=@n2), B=Multiply(A=@n3, B=@n4)), B=NamedRerouteUsage())
  n6 = NamedRerouteUsage()
  n7 = NamedRerouteUsage()
  n8 = NamedRerouteUsage()
  n9 = NamedRerouteUsage()
  n10 = Multiply(A=Add(A=Multiply(A=@n6, B=@n7), B=Multiply(A=@n8, B=@n9)), B=NamedRerouteUsage())
  FunctionOutput(None=NamedRerouteDeclaration(None=Add(A=@n5, B=@n10)) {name=iHash, nodeColor=<rgba 0 0.25 1 1>})
  NamedRerouteDeclaration(None=FunctionInput(Preview=TextureCoordinate()) {bUsePreviewValueAsDefault=true, inputName=IN_V, inputType=FunctionInput_Vector2}) {name=IN_V, nodeColor=<rgba 0 1 0.3203 1>}
  n11 = TextureCoordinate()
  NamedRerouteDeclaration(None=FunctionInput(Preview=Add(A=@n11) {constB=0}) {bUsePreviewValueAsDefault=true, inputName=IN_R, inputType=FunctionInput_Vector2, sortPriority=1}) {name=IN_R, nodeColor=<rgba 1 0 0 1>}
  NamedRerouteDeclaration(None=OneMinus(None=NamedRerouteUsage())) {name=ip_x_invert, nodeColor=<rgba 0 0.7187 1 1>}
  NamedRerouteDeclaration(None=Multiply(A=NamedRerouteUsage(), B=NamedRerouteUsage())) {name=VmultR, nodeColor=<rgba 1 0 0.9609 1>}
  n12 = Floor(None=Add(A=NamedRerouteUsage(), B=Constant2Vector()))
  n13 = NamedRerouteUsage()
  NamedRerouteDeclaration(None=MaterialFunctionCall!(IN_2dVector=Divide(A=@n12, B=@n13)) {materialFunction=&/MF_BisonDrainHash}.OUT_HashFloat) {name=h00, nodeColor=<rgba 0 1 0.7188 1>}
  n14 = Floor(None=Add(A=NamedRerouteUsage(), B=Constant2Vector() {r=1}))
  n15 = NamedRerouteUsage()
  NamedRerouteDeclaration(None=MaterialFunctionCall(IN_2dVector=Divide(A=@n14, B=@n15)) {materialFunction=&MF_BisonDrainHash}.OUT_HashFloat) {name=h10, nodeColor=<rgba 0.4192 0.0047 1 1>}
  n16 = Floor(None=Add(A=NamedRerouteUsage(), B=Constant2Vector() {g=1}))
  n17 = NamedRerouteUsage()
  NamedRerouteDeclaration(None=MaterialFunctionCall(IN_2dVector=Divide(A=@n16, B=@n17)) {materialFunction=&MF_BisonDrainHash}.OUT_HashFloat) {name=h01, nodeColor=<rgba 1 0.1202 0.0022 1>}
  n18 = Floor(None=Add(A=NamedRerouteUsage(), B=Constant2Vector() {g=1, r=1}))
  n19 = NamedRerouteUsage()
  NamedRerouteDeclaration(None=MaterialFunctionCall(IN_2dVector=Divide(A=@n18, B=@n19)) {materialFunction=&MF_BisonDrainHash}.OUT_HashFloat) {name=h11, nodeColor=<rgba 0.7101 1 0 1>}
  n20 = Subtract(A=NamedRerouteUsage(), B=Floor(None=NamedRerouteUsage()))
  n21 = MaterialFunctionCall(Float2=NamedRerouteDeclaration(None=@n20) {name=ip, nodeColor=<rgba 1 0 0.5625 1>}) {materialFunction=&BreakOutFloat2Components}
  NamedRerouteDeclaration(None=@n21.R) {name=ip_x, nodeColor=<rgba 0 0.8828 1 1>}
  NamedRerouteDeclaration(None=@n21.G) {name=ip_y, nodeColor=<rgba 1 0.7969 0 1>}
  OneMinus(None=NamedRerouteUsage())
  SmoothStep(Min=Constant2Vector(), Max=Constant2Vector() {g=1, r=1}, Value=@n20)
 ]}
