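# Material function for the BisonDrain TV gadget prop materials that yields one hash term, apparently for chaining into a summed series.
# It offsets the UV input (texture coordinates by default) by the scalar index on both axes and feeds it to MF_iHash with R fixed at one.
# It also outputs the incremented index.
# A MF_PowerMultBisonDrain call is present but feeds no output.
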MaterialFunction /BRCosmetics/Animation/Props/Gadgets_BisonDrain_TV/Materials/MF_iHashForSumming
graph {nodes [
  n1 = NamedRerouteUsage()
  n2 = NamedRerouteUsage()
  n3 = AppendVector(A=@n2, B=@n2)
  FunctionOutput(None=MaterialFunctionCall!(IN_V=Add(A=@n1, B=@n3), IN_R=Constant() {r=1}) {materialFunction=&/MF_iHash}.Result)
  n4 = FunctionInput(Preview=Constant()) {inputName=IN_i, inputType=FunctionInput_Scalar}
  FunctionOutput(None=Add(A=NamedRerouteDeclaration(None=@n4) {name=IN_i, nodeColor=<rgba 0 0.1797 1 1>})) {outputName=i_incremented, sortPriority=1}
  MaterialFunctionCall!(i=NamedRerouteUsage()) {materialFunction=&/MF_PowerMultBisonDrain}
  NamedRerouteDeclaration(None=FunctionInput(Preview=TextureCoordinate()) {bUsePreviewValueAsDefault=true, inputName=IN_V, inputType=FunctionInput_Vector2}) {name=IN_V, nodeColor=<rgba 0 1 0.3203 1>}
 ]}
